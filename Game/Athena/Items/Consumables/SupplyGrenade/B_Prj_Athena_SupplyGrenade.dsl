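; Supply grenade projectile Blueprint that explodes when it comes to a stop and calls in an Athena supply drop.
; Cosmetic construction attaches rotating movement to the mesh and matches the overlap capsule to the collision capsule; stopping clears the rotation and triggers the explosion.
; On explosion, authority spawns the drop at the grenade, while remotes spawn a temp effect aligned to the stop hit's normal.
; SetTeleportDestination stores an offset impact point but is not called here.
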(event EventOnStop
  (params (Hit :type Hit%20Result%20Structure%20%28by%20ref%29))
  ;[0.7116 0.838 1] OnStop - Stop the rotation
  (CallFunction|StopRotation)
  (Delivery|DoExplosion))

(event Custom|Stop_Rotation
  ;[1 1 1] Stop Rotating Movement
  (Components|Movement|SetUpdatedComponent (Variables|Default|GetRotatingMovement) 0))

(event EventOnExploded
  (params (HitActors :type Array%20of%20Actor%20Object%20References) (HitResults :type Array%20of%20Hit%20Result%20Structures))
  ;[1 1 1] OnExploded
  (Networking|SwitchHasAuthority
    (:Authority
      ;[1 1 1] Spawn Supply Drop
      (Game|SpawnActorfromClass "/Game/Athena/SupplyDrops/AthenaSupplyDrop_02_SupplyGrenade.AthenaSupplyDrop_02_SupplyGrenade_C" (Math|Transform|MakeTransform (Transformation|GetActorLocation)) "Undefined" "OverrideRootScale" 0 0 0))
    (:Remote
      ;[1 1 1] VisFX
      (bind _returnvalue (Effects|SpawnEmitteratLocation "/Game/Athena/Items/Consumables/SupplyGrenade/P_SupplyGrenade_Explosion_TEMP.P_SupplyGrenade_Explosion_TEMP" (Transformation|GetActorLocation) (Math|Conversions|RotationFromXVector (.ImpactNormal (Collision|BreakHitResult (stale EventOnStop Hit)))) "0.500000,0.500000,0.500000"))
      (Rendering|SetTranslucentSortPriority _returnvalue -1))))

(fn SetTeleportDestination
  (params (Hit :type Hit%20Result%20Structure%20%28by%20ref%29))
  (bind (_bblockinghit _binitialoverlap _time _distance _location _impactpoint _normal _impactnormal) (Collision|BreakHitResult Hit))
  (Networking|SwitchHasAuthority
    (:Authority
      (Variables|Default|SetTeleportDestination (+ _impactpoint (* _impactnormal 100))))))

(fn ConstructionScript
  (Networking|CanExecuteCosmeticEvents
    (:True
      (Components|Movement|SetUpdatedComponent (Variables|Default|GetRotatingMovement) (Variables|B_Prj_Grenade|GetMesh))
      (Components|Capsule|SetCapsuleSize (Variables|Default|GetOverlapCapsule) (Variables|/Script/Engine.CapsuleComponent|GetCapsuleRadius (Variables|Collision|GetCapsuleComponent)) (Variables|/Script/Engine.CapsuleComponent|GetCapsuleHalfHeight (Variables|Collision|GetCapsuleComponent)) true))))
